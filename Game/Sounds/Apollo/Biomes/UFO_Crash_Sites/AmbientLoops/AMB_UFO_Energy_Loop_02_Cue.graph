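# UFO crash site ambient Sound Cue that loops an alien energy drone in the Apollo biome.
# A single looping Wave Player feeds the root directly, with no randomization or modulation.
# It uses the shared looping spot attenuation, a boosted volume and the Creative gadgets sound class.
# A concurrency override prevents new instances from starting once the limit is reached.
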SoundCue /Game/Sounds/Apollo/Biomes/UFO_Crash_Sites/AmbientLoops/AMB_UFO_Energy_Loop_02_Cue
root=WavePlayer_2
nodes [{
  id=WavePlayer_2
  properties {bLooping=true, soundWaveAssetPtr=&/AMB_UFO_Crashsite_Loop_Alien_Drone_02}
 }]
properties {
 attenuationSettings=/Game/Sounds/Attenuation/Fort_Looping_Spot_Att
 bOverrideConcurrency=true
 maxDistance=3750
 soundClassObject=&Fort_Gadgets_Creative
 volumeMultiplier=2
 concurrencyOverrides {resolutionRule=PreventNew}
}
