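# Sound Cue for a looping UFO crash site energy ambience in the Apollo biome audio.
# A single looping Wave Player, wired straight to the output, plays the Alien Brain Scan crash site loop wave.
# It reuses the Directional Launcher ambient attenuation asset and routes to the Creative gadgets sound class.
# Overridden concurrency uses the prevent-new rule, so playing instances are kept over new ones once the limit is reached.
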SoundCue /Game/Sounds/Apollo/Biomes/UFO_Crash_Sites/AmbientLoops/AMB_UFO_Energy_Loop_03_Cue
root=WavePlayer_0
nodes [{
  id=WavePlayer_0
  properties {bLooping=true, soundWaveAssetPtr=&/AMB_UFO_Crashsite_Loop_Alien_Brain_Scan}
 }]
properties {
 attenuationSettings=/Game/Athena/Sounds/LaunchPad/DirectionalLauncher/DirectionalLauncher_Ambient_Att
 bOverrideConcurrency=true
 maxDistance=2700
 soundClassObject=&Fort_Gadgets_Creative
 volumeMultiplier=1
 concurrencyOverrides {resolutionRule=PreventNew}
}
