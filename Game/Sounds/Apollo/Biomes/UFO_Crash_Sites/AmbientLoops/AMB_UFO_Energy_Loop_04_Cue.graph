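# Looping ambient sound cue for UFO crash site energy in the Apollo biome.
# A single looping wave player feeds the root with a tech ambience loop, played at reduced volume and pitch.
# It borrows the directional launch pad's ambient attenuation and the creative gadgets sound class.
# Overridden concurrency prevents new instances from starting once the limit is reached.
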SoundCue /Game/Sounds/Apollo/Biomes/UFO_Crash_Sites/AmbientLoops/AMB_UFO_Energy_Loop_04_Cue
root=WavePlayer_1
nodes [{
  id=WavePlayer_1
  properties {bLooping=true, soundWaveAssetPtr=&/AMB_UFO_Crashsite_Ambience_Tech_Loop_01}
 }]
properties {
 attenuationSettings=/Game/Athena/Sounds/LaunchPad/DirectionalLauncher/DirectionalLauncher_Ambient_Att
 bOverrideConcurrency=true
 maxDistance=2700
 pitchMultiplier=0.8
 soundClassObject=&Fort_Gadgets_Creative
 volumeMultiplier=0.8
 concurrencyOverrides {resolutionRule=PreventNew}
}
