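# AMB_UFO_Energy_Loop_05_Cue is a sound cue for a looping tech ambience at Apollo's UFO crash site biome.
# A single looping wave player plays the Tech Loop 02 crash site ambience wave at reduced volume.
# It borrows the trap loop attenuation and the Creative gadgets sound class.
# A concurrency override prevents new instances from starting once the limit is reached.
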SoundCue /Game/Sounds/Apollo/Biomes/UFO_Crash_Sites/AmbientLoops/AMB_UFO_Energy_Loop_05_Cue
root=WavePlayer_0
nodes [{
  id=WavePlayer_0
  properties {bLooping=true, soundWaveAssetPtr=&/AMB_UFO_Crashsite_Ambience_Tech_Loop_02}
 }]
properties {
 attenuationSettings=/Game/Sounds/Attenuation/Fort_Traps_Loop_Att
 bOverrideConcurrency=true
 maxDistance=2500
 soundClassObject=&Fort_Gadgets_Creative
 volumeMultiplier=0.3
 concurrencyOverrides {resolutionRule=PreventNew}
}
